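# Physics asset for the hood of the medium female player's red Street Fashion outfit, for simulating hood motion on the upper back.
# Left and right hood bones each carry an identical box body.
# Spine_05 carries a capsule plus a rotated box as the torso volume.
# Each hood bone is constrained directly to spine_05 with moderate, asymmetric swing limits and very tight twist, so the halves swing independently with little roll.
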PhysicsAsset /Game/Characters/Player/Female/Medium/Bodies/F_MED_Street_Fashion_Red/Meshes/F_MED_Street_Fashion_Red_Hood_Physics
bodies {
 dyn_hood_l [{shapeType=Box, extentX=7.1888, extentY=7.4266, extentZ=2.7443, center=<2.7373 0 -0.0002>}]
 dyn_hood_r [{shapeType=Box, extentX=7.1888, extentY=7.4266, extentZ=2.7443, center=<2.7373 0 -0.0002>}]
 spine_05 [
  {shapeType=Capsule, length=21.25, radius=8.75, center=<0.8141 1.5922 -0.0002>}
  {shapeType=Box, extentX=13, extentY=5, extentZ=20, center=<-4.8728 -4.7031 0.0002>, rotation=<rot 0 4.8303 0>}
 ]
}
constraints [
 {bone1Name=dyn_hood_r, bone2Name=spine_05, swing1LimitDegrees=35, swing2LimitDegrees=60, twistLimitDegrees=3}
 {bone1Name=dyn_hood_l, bone2Name=spine_05, swing1LimitDegrees=35, swing2LimitDegrees=60, twistLimitDegrees=3}
]
