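# Post-process style cosmetic AnimBP for the F_MED_Street_Fashion_Red female body, adding secondary motion to jacket, hood, drawstrings and grenade.
# Copies the parent mesh pose, applies bone constraint and thigh/hood scale correctives driven by leg swing and head yaw, and simulates strings with AnimDynamics.
# The cached result feeds two RigidBody chains, selected by backbling presence to swap physics setups, then nudges the front jacket bone.
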AnimBlueprint /Game/Characters/Player/Female/Medium/Bodies/F_MED_Street_Fashion_Red/Meshes/F_MED_Street_Fashion_Red_AnimBP
target_skeleton=&/F_MED_Street_Fashion_Red_Skeleton
graphs [
 AnimGraph {nodes [
   RigidBody!(ComponentPose, Alpha, OverrideWorldGravity:Vector="0,0,0", ExternalForce) -> (Pose) {
    bClampLinearTranslationLimitToRefPose=true
    bOverrideWorldGravity=true
    overridePhysicsAsset=&/F_MED_Street_Fashion_Red_Physics
    simulationSpace=BaseBoneSpace
    alphaBoolBlend {blendOption=Linear}
    baseBoneRef {boneName=pelvis}
    componentAppliedLinearAccClamp=<6000 6000 6000>
    componentLinearAccScale=<0.5 0.5 0.5>
    componentLinearVelScale=<0.5 0.5 0.5>
   }
   RigidBody!(ComponentPose, Alpha=0.5, ExternalForce) -> (Pose) {
    bClampLinearTranslationLimitToRefPose=true
    bOverrideWorldGravity=true
    overridePhysicsAsset=&/F_MED_Street_Fashion_Red_String_Physics
    simulationSpace=BaseBoneSpace
    alphaBoolBlend {blendOption=Linear}
    baseBoneRef {boneName=root}
    componentAppliedLinearAccClamp=<6000 6000 6000>
    componentLinearAccScale=<0.5 0.5 0.5>
    componentLinearVelScale=<0.5 0.5 0.5>
    overrideWorldGravity {z=-980}
   }
   RigidBody() {
    bClampLinearTranslationLimitToRefPose=true
    bOverrideWorldGravity=true
    overridePhysicsAsset=&/F_MED_Street_Fashion_Red_PackEquipped_Physics
    simulationSpace=BaseBoneSpace
    alphaBoolBlend {blendOption=Linear}
    baseBoneRef {boneName=pelvis}
    componentAppliedLinearAccClamp=<6000 6000 6000>
    componentLinearAccScale=<0.5 0.5 0.5>
    componentLinearVelScale=<0.5 0.5 0.5>
    overrideWorldGravity {z=-980}
   }
   RigidBody() {
    bClampLinearTranslationLimitToRefPose=true
    bOverrideWorldGravity=true
    overridePhysicsAsset=&/F_MED_Street_Fashion_Red_Hood_Physics
    simulationSpace=BaseBoneSpace
    alphaBoolBlend {blendOption=Linear}
    baseBoneRef {boneName=spine_02}
    componentAppliedLinearAccClamp=<5000 5000 5000>
    componentLinearAccScale=<0.2 0.2 0.2>
    componentLinearVelScale=<0.2 0.2 0.2>
    overrideWorldGravity {z=-980}
   }
   n1 = RigidBody!(ComponentPose=Local To Component(LocalPose=Use cached pose 'JacketCache'()), Alpha, OverrideWorldGravity:Vector="0,0,-980", ExternalForce) -> (Pose) {
    bClampLinearTranslationLimitToRefPose=true
    bOverrideWorldGravity=true
    overridePhysicsAsset=&F_MED_Street_Fashion_Red_PackEquipped_Physics
    simulationSpace=BaseBoneSpace
    alphaBoolBlend {blendOption=Linear}
    baseBoneRef {boneName=pelvis}
    componentAppliedLinearAccClamp=<6000 6000 6000>
    componentLinearAccScale=<0.3 0.3 0.3>
    componentLinearVelScale=<0.3 0.3 0.3>
    overrideWorldGravity {z=-980}
    simSpaceSettings {
     maxAngularAcceleration=500
     maxAngularVelocity=3
     maxLinearAcceleration=400
     maxLinearVelocity=250
     velocityScaleZ=0
     worldAlpha=1
     externalLinearDragV=<0.1 0.1>
    }
   }
   n2 = RigidBody!(ComponentPose=@n1, Alpha=0, ExternalForce) -> (Pose) {
    bClampLinearTranslationLimitToRefPose=true
    bOverrideWorldGravity=true
    overridePhysicsAsset=&F_MED_Street_Fashion_Red_Hood_Physics
    simulationSpace=BaseBoneSpace
    alphaBoolBlend {blendOption=Linear}
    baseBoneRef {boneName=spine_02}
    componentAppliedLinearAccClamp=<5000 5000 5000>
    componentLinearAccScale=<0.2 0.2 0.2>
    componentLinearVelScale=<0.2 0.2 0.2>
    overrideWorldGravity {z=-980}
   }
   n3 = Component To Local(ComponentPose=@n2)
   n4 = RigidBody!(ComponentPose=Local To Component(LocalPose=Use cached pose 'JacketCache'()), Alpha, OverrideWorldGravity:Vector="0,0,0", ExternalForce) -> (Pose) {
    bClampLinearTranslationLimitToRefPose=true
    bOverrideWorldGravity=true
    overridePhysicsAsset=&F_MED_Street_Fashion_Red_Physics
    simulationSpace=BaseBoneSpace
    alphaBoolBlend {blendOption=Linear}
    baseBoneRef {boneName=pelvis}
    componentAppliedLinearAccClamp=<6000 6000 6000>
    componentLinearAccScale=<0.3 0.3 0.3>
    componentLinearVelScale=<0.3 0.3 0.3>
    simSpaceSettings {
     maxAngularAcceleration=500
     maxAngularVelocity=3
     maxLinearAcceleration=400
     maxLinearVelocity=250
     velocityScaleZ=0
     worldAlpha=1
     externalLinearDragV=<0.1 0.1>
    }
   }
   n5 = RigidBody(ComponentPose=@n4, OverrideWorldGravity="0,0,0") {
    bClampLinearTranslationLimitToRefPose=true
    bOverrideWorldGravity=true
    overridePhysicsAsset=&F_MED_Street_Fashion_Red_Hood_Physics
    simulationSpace=BaseBoneSpace
    alphaBoolBlend {blendOption=Linear}
    baseBoneRef {boneName=spine_02}
    componentAppliedLinearAccClamp=<5000 5000 5000>
    componentLinearAccScale=<0.2 0.2 0.2>
    componentLinearVelScale=<0.2 0.2 0.2>
   }
   n6 = Component To Local(ComponentPose=@n5)
   n7 = Get Is_Wearing_Backbling() -> (Is_Wearing_Backbling:bool)
   n8 = "Transform (Modify) Bone - Bone: dyn_jacket_front"(
    ComponentPose=Local To Component(LocalPose=Blend Poses by bool(BlendPose_0=@n3, BlendPose_1=@n6, bActiveValue=@n7) {blendType=Linear})
    Alpha=1
    Translation="0,1,0"
    Rotation="0,0,0"
    Scale="1,1,1"
   ) {
    scaleMode=BMM_Additive
    scaleSpace=BCS_BoneSpace
    translationMode=BMM_Additive
    translationSpace=BCS_BoneSpace
    alphaBoolBlend {blendOption=Linear}
    boneToModify {boneName=dyn_jacket_front}
    translation {y=1}
   }
   Output Pose(Result=Component To Local(ComponentPose=@n8)) {layerGroup=None}
   n9 = "Constraint - dyn_grenade (PPPP)"!(
    ComponentPose=Local To Component(LocalPose=Copy Pose From Mesh() {bUseAttachedParent=true})
    Alpha=1
    ConstraintWeights_0=0.4754
    ConstraintWeights_1:float=0.0527
    ConstraintWeights_2:float=0.0007
    ConstraintWeights_3:float=0.4712
   ) -> (Pose) {
    alphaBoolBlend {blendOption=Linear}
    boneToModify {boneName=dyn_grenade}
    constraintSetup [
     {targetBone {boneName=thigh_r}}
     {targetBone {boneName=pelvis}}
     {targetBone {boneName=thigh_twist_01_l}}
     {targetBone {boneName=thigh_twist_01_r}}
    ]
    constraintWeights [1,1,1,1]
   }
   n10 = "Transform (Modify) Bone - Bone: dyn_subthigh_l"(ComponentPose=@n9, Alpha=Get legSwingDelta_l() -> (legSwingDelta_l:double), Translation="0,0,0", Rotation="0,0,0", Scale="4,1,4") {
    scaleMode=BMM_Additive
    scaleSpace=BCS_BoneSpace
    translationMode=BMM_Additive
    translationSpace=BCS_BoneSpace
    alphaBoolBlend {blendOption=Linear}
    boneToModify {boneName=dyn_subthigh_l}
    scale {x=4, z=4}
   }
   n11 = "Transform (Modify) Bone - Bone: dyn_subthigh_r"(ComponentPose=@n10, Alpha=Get legSwingDelta_r() -> (legSwingDelta_r:double), Translation="0,0,0", Rotation="0,0,0", Scale="2.25,1,2.25") {
    scaleMode=BMM_Additive
    scaleSpace=BCS_BoneSpace
    translationMode=BMM_Additive
    translationSpace=BCS_BoneSpace
    alphaBoolBlend {blendOption=Linear}
    boneToModify {boneName=dyn_subthigh_r}
    scale {x=2.25, z=2.25}
   }
   n12 = "Transform (Modify) Bone - Bone: dyn_string_r_02"!(ComponentPose=@n11, bAlphaBoolEnabled:bool=True, Translation="0,0,0", Rotation="-6,0,0", Scale="1,1,1") -> (Pose) {
    alphaInputType=Bool
    rotationMode=BMM_Additive
    rotationSpace=BCS_BoneSpace
    scaleMode=BMM_Additive
    translationMode=BMM_Additive
    alphaBoolBlend {blendOption=Linear}
    boneToModify {boneName=dyn_string_r_02}
    rotation {pitch=-6}
   }
   n13 = "Transform (Modify) Bone - Bone: dyn_string_l_02"!(ComponentPose=@n12, bAlphaBoolEnabled:bool=True, Translation="0,0,0", Rotation="-6,0,0", Scale="1,1,1") -> (Pose) {
    alphaInputType=Bool
    rotationMode=BMM_Additive
    rotationSpace=BCS_BoneSpace
    scaleMode=BMM_Additive
    scaleSpace=BCS_BoneSpace
    translationMode=BMM_Additive
    translationSpace=BCS_BoneSpace
    alphaBoolBlend {blendOption=Linear}
    boneToModify {boneName=dyn_string_l_02}
    rotation {pitch=-6}
   }
   n14 = Anim Dynamics - Bone: dyn_string_l_01(ComponentPose=@n13, Alpha=1, ExternalForce="0,0,0") {
    angularDampingOverride=1
    bLinearSpring=true
    bOverrideAngularDamping=true
    bUseGravityOverride=true
    linearSpringConstant=800
    simulationSpace=RootRelative
    alphaBoolBlend {blendOption=Linear}
    boundBone {boneName=dyn_string_l_01}
    componentAppliedLinearAccClamp=<10000 10000 10000>
    componentLinearAccScale=<0.1 0.4 0.55>
    componentLinearVelScale=<0.1 0.4 0.55>
    gravityOverride {z=-980}
    physicsBodyDefinitions [{
      boundBone {boneName=dyn_string_l_01}
      boxExtents=<6 6 6>
      constraintSetup {angularLimitsMax=<10 0 15>, angularLimitsMin=<-10 -50 -15>, angularTarget=<1 0 0>}
      localJointOffset=<0.35 0 0>
     }]
   }
   n15 = Anim Dynamics - Bone: dyn_string_l_02(ComponentPose=@n14, Alpha=1, ExternalForce="0,0,0") {
    angularDampingOverride=1
    bLinearSpring=true
    bOverrideAngularDamping=true
    bUseGravityOverride=true
    linearSpringConstant=800
    alphaBoolBlend {blendOption=Linear}
    boundBone {boneName=dyn_string_l_02}
    componentAppliedLinearAccClamp=<10000 10000 10000>
    componentLinearAccScale=<0.1 0.4 0.55>
    componentLinearVelScale=<0.1 0.4 0.55>
    gravityOverride {z=-980}
    physicsBodyDefinitions [{
      sphereCollisionRadius=3
      boundBone {boneName=dyn_string_l_02}
      boxExtents=<6 6 6>
      constraintSetup {angularLimitsMax=<8 5 10>, angularLimitsMin=<-8 -25 -10>, angularTarget=<1 0 0>}
      localJointOffset=<0.35 0 0>
     }]
    relativeSpaceBone {boneName=dyn_string_l_01}
   }
   n16 = Anim Dynamics - Bone: dyn_string_l_03(ComponentPose=@n15, Alpha=1, ExternalForce="0,0,0") {
    angularDampingOverride=1
    bLinearSpring=true
    bOverrideAngularDamping=true
    bUseGravityOverride=true
    linearSpringConstant=800
    alphaBoolBlend {blendOption=Linear}
    boundBone {boneName=dyn_string_l_03}
    componentAppliedLinearAccClamp=<10000 10000 10000>
    componentLinearAccScale=<0.1 0.4 0.55>
    componentLinearVelScale=<0.1 0.4 0.55>
    gravityOverride {z=-980}
    physicsBodyDefinitions [{
      boundBone {boneName=dyn_string_l_03}
      boxExtents=<6 6 6>
      constraintSetup {angularLimitsMax=<5 5 5>, angularLimitsMin=<-5 -15 -5>, angularTarget=<1 0 0>}
      localJointOffset=<0.35 0 0>
     }]
    relativeSpaceBone {boneName=dyn_string_l_01}
   }
   n17 = Anim Dynamics - Bone: dyn_string_r_01(ComponentPose=@n16, Alpha=1, ExternalForce="0,0,0") {
    angularDampingOverride=1
    bLinearSpring=true
    bOverrideAngularDamping=true
    bUseGravityOverride=true
    linearSpringConstant=800
    simulationSpace=RootRelative
    alphaBoolBlend {blendOption=Linear}
    boundBone {boneName=dyn_string_r_01}
    componentAppliedLinearAccClamp=<10000 10000 10000>
    componentLinearAccScale=<0.1 0.8 0.53>
    componentLinearVelScale=<0.1 0.8 0.53>
    gravityOverride {z=-980}
    physicsBodyDefinitions [{
      boundBone {boneName=dyn_string_r_01}
      boxExtents=<5 5 5>
      constraintSetup {angularLimitsMax=<10 0 20>, angularLimitsMin=<-10 -50 -20>, angularTarget=<1 0 0>}
      localJointOffset=<0.35 0 0>
     }]
   }
   n18 = Anim Dynamics - Bone: dyn_string_r_02(ComponentPose=@n17, Alpha=1, ExternalForce="0,0,0") {
    angularDampingOverride=1
    bLinearSpring=true
    bOverrideAngularDamping=true
    bUseGravityOverride=true
    linearSpringConstant=800
    alphaBoolBlend {blendOption=Linear}
    boundBone {boneName=dyn_string_r_02}
    componentAppliedLinearAccClamp=<10000 10000 10000>
    componentLinearAccScale=<0.1 0.8 0.53>
    componentLinearVelScale=<0.1 0.8 0.53>
    gravityOverride {z=-980}
    physicsBodyDefinitions [{
      sphereCollisionRadius=3
      boundBone {boneName=dyn_string_r_02}
      boxExtents=<5 5 5>
      constraintSetup {angularLimitsMax=<8 0 15>, angularLimitsMin=<-8 -30 -15>, angularTarget=<1 0 0>}
      localJointOffset=<0.35 0 0>
     }]
    relativeSpaceBone {boneName=dyn_string_l_01}
   }
   n19 = Anim Dynamics - Bone: dyn_string_r_03(ComponentPose=@n18, Alpha=1, ExternalForce="0,0,0") {
    angularDampingOverride=1
    bLinearSpring=true
    bOverrideAngularDamping=true
    bUseGravityOverride=true
    linearSpringConstant=800
    alphaBoolBlend {blendOption=Linear}
    boundBone {boneName=dyn_string_r_03}
    componentAppliedLinearAccClamp=<10000 10000 10000>
    componentLinearAccScale=<0.1 0.8 0.53>
    componentLinearVelScale=<0.1 0.8 0.53>
    gravityOverride {z=-980}
    physicsBodyDefinitions [{
      boundBone {boneName=dyn_string_r_03}
      boxExtents=<5 5 5>
      constraintSetup {angularLimitsMax=<5 5 10>, angularLimitsMin=<-5 -20 -10>, angularTarget=<1 0 0>}
      localJointOffset=<0.35 0 0>
     }]
    relativeSpaceBone {boneName=dyn_string_l_01}
   }
   n20 = "Transform (Modify) Bone - Bone: dyn_hood_l"(ComponentPose=@n19, Alpha=Get head_rot_back() -> (head_rot_back:double), Translation="0,0,0", Rotation="0,0,0", Scale="1.25,1.25,1.5") {
    scaleMode=BMM_Additive
    scaleSpace=BCS_BoneSpace
    translationMode=BMM_Additive
    translationSpace=BCS_BoneSpace
    alphaBoolBlend {blendOption=Linear}
    boneToModify {boneName=dyn_hood_l}
    scale=<1.25 1.25 1.5>
   }
   Save cached pose 'JacketCache'(Pose=Component To Local(ComponentPose=@n20)) {cacheName=JacketCache}
   Get isBackStatic() -> (isBackStatic:bool)
  ]}
 EventGraph {nodes [
   Event BlueprintUpdateAnimation()
   > n1 = Sequence!(execute) -> (then_0, then_1, then_2:Exec, then_3:Exec, then_4:Exec)
   n2 = Get player_pawn() -> (player_pawn:Fort Player Pawn Athena Object Reference)
   > Is Valid(InputObject=@n2)
   > Do Once()
   > Set isBackpackEquipped(isBackpackEquipped:bool="Not Equal (Object)"!(A:Object Reference=GetSkeletalMeshForPartType(self=@n2, PartType=Backpack), B:Object Reference) -> (ReturnValue:bool)) -> (Output_Get:bool)
   n3 = Gravity Override State Function DEPRECATED(
    execute=@n1.then_4
    CharacterPartAnimInstance=Self-Reference()
    GameTimeInSeconds=GetGameTimeInSeconds()
    GravityOverrideParams=Get Gravity Override Params()
    DebugJointName=dyn_hackBottomArmor
   )
   > Set Override World Gravity(Override World Gravity:Vector=@n3) -> (Output_Get:Vector)
   n4 = GetOwnerSkeletalMeshComponent()
   n5 = BreakTransform(InTransform=GetDeltaTransformFromRefPose(self=@n4, BoneName=head, BaseName=spine_05))
   Set head_rot_back(execute=@n1.then_2, head_rot_back:double=MapRangeClamped(Value=BreakRotator(InRot=@n5.Rotation).Yaw, InRangeA=0, InRangeB=-48)) -> (Output_Get:double)
   n6 = "integer > integer"(A=Get CurrentLOD(), B=1) {is=title K2Node_CallFunction}
   n7 = OR Boolean(A=NOT Boolean(A=Get bRigidBodyEnabled()), B=@n6)
   Set isRigidBodyDisabled(execute=@n1.then_1, isRigidBodyDisabled:bool=@n7) -> (Output_Get:bool)
   > Set isBackStatic(isBackStatic:bool=OR Boolean(A=Get isBackpackEquipped() -> (isBackpackEquipped:bool), B=@n7)) -> (Output_Get:bool)
   > Set Rigids_lod_offset(Rigids_lod_offset:Vector=SelectVector(A="0,-5, 0", B="0, 0, 0", bPickA=@n6)) -> (Output_Get:Vector)
   > Set Rigids_lod_rot_offset(Rigids_lod_rot_offset:Rotator=SelectRotator!(A:Rotator="0,20, 0", B:Rotator="0, 0, 0", bPickA:bool=@n6) -> (ReturnValue:Rotator)) -> (Output_Get:Rotator)
   Set legSwingDelta_r(
    execute=@n1.then_3
    legSwingDelta_r:double=MapRangeClamped(Value=GetTwistAndSwingAngleOfDeltaRotationFromRefPose(self=@n4, BoneName=thigh_r).OutSwingAngle, InRangeA=150, InRangeB=75)
   ) -> (Output_Get:double)
   > Set legSwingDelta_l(legSwingDelta_l:double=MapRangeClamped(Value=GetTwistAndSwingAngleOfDeltaRotationFromRefPose(self=@n4, BoneName=thigh_l).OutSwingAngle, InRangeA=43, InRangeB=110)) -> (Output_Get:double)
   Event BlueprintInitializeAnimation()
   n8 = TryGetPawnOwner()
   > Is Valid(InputObject=@n8)
   > n9 = Cast To FortPlayerPawnAthena(Object=@n8)
   > Set player_pawn(player_pawn:Fort Player Pawn Athena Object Reference=@n9) -> (Output_Get:Fort Player Pawn Athena Object Reference)
   Event OnCharacterCustomizationLoaded()
   n10 = CheckForBackbling()
   n11 = OR Boolean!(A=@n10, B=@n10.bIsBackblingACape, C:bool=@n10.bIsBackblingLong) -> (ReturnValue)
   > Set Is_Wearing_Backbling(Is_Wearing_Backbling:bool=@n11) -> (Output_Get:bool)
   > Set BackBling_alpha(BackBling_alpha:double="To Float (Boolean)"(InBool=@n11)) -> (Output_Get:double)
   TryGetPawnOwner()
  ]}
]
variables [
 BackBling_alpha {type=number}
 Gravity Override Params {type=GravityOverrideParamsStruct}
 GravityOverride {type=Vector}
 head_rot_back {type=number}
 Is_Wearing_Backbling {type=boolean}
 isBackpackEquipped {type=boolean}
 isBackStatic {type=boolean}
 isRigidBodyDisabled {type=boolean}
 legSwingDelta_l {type=number}
 legSwingDelta_r {type=number}
 Override World Gravity {type=Vector}
 player_pawn {type=FortPlayerPawnAthena}
 Rigids_lod_offset {type=Vector}
 Rigids_lod_rot_offset {type=Rotator}
]
