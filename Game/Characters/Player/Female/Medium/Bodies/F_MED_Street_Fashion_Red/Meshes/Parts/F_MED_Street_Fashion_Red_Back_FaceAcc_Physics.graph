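# Physics asset for the female medium Street Fashion Red back/face accessory part, setting up hair bone physics.
# Four dyn_hair chains (mid-left, back left, middle, right) hang from the head through swing- and twist-limited constraints with very tight twist.
# Capsules on head, neck, upper spine and clavicles provide body shapes around the hair, with clavicles rotating freely against spine_05.
# The small dyn_hair_bk_r_base box and neck body have no constraints.
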PhysicsAsset /Game/Characters/Player/Female/Medium/Bodies/F_MED_Street_Fashion_Red/Meshes/Parts/F_MED_Street_Fashion_Red_Back_FaceAcc_Physics
bodies {
 clavicle_l [{shapeType=Capsule, length=12, radius=5.5, shapeName=clavicle_l_capsule, center=<3.5 -1 -1>, rotation=<rot 80 -0.0003 -0.0003>}]
 clavicle_r [{shapeType=Capsule, length=12, radius=5.5, shapeName=clavicle_r_capsule, center=<-3.5 1 1>, rotation=<rot 80 -0.0003 -0.0004>}]
 dyn_hair_bk_02 [{
   shapeType=Capsule
   length=5
   radius=1.75
   shapeName=dyn_hair_bk_02_capsule
   center=<4.4511 1.2932 -0.2195>
   rotation=<rot 79.9972 -166.5116 180>
  }]
 dyn_hair_bk_l_01 [{shapeType=Box, extentX=7.8766, extentY=5.4314, extentZ=3.2208, center=<3.3885 0 -0.556>}]
 dyn_hair_bk_l_02 [{shapeType=Box, extentX=6.8766, extentY=5.1814, extentZ=3.7208, center=<3.4338 0 -0.0701>}]
 dyn_hair_bk_l_parent [{shapeType=Box, extentX=3.8766, extentY=5.4314, extentZ=1.2208, center=<1.8725 -0.0001 0>}]
 dyn_hair_bk_mid_01 [{
   shapeType=Capsule
   length=3.8092
   radius=1.7147
   shapeName=dyn_hair_bk_mid_01_capsule
   center=<2.7556 0.5482 -0.8666>
   rotation=<rot 67.3018 -67.4891 -69.7649>
  }]
 dyn_hair_bk_mid_parent [{shapeType=Box, extentX=4.8766, extentY=3.9314, extentZ=1.2208, center=<2.296 0.0125 0>, rotation=<rot 0 7.025 15.0536>}]
 dyn_hair_bk_r_01 [{shapeType=Capsule, length=3.3295, radius=2, shapeName=dyn_hair_bk_r_01_capsule, center=<2.6847 -0.8981 -1.098>, rotation=<rot 90 -180 180>}]
 dyn_hair_bk_r_02 [{shapeType=Capsule, length=3.4749, radius=2.0251, shapeName=dyn_hair_bk_r_02_capsule, center=<4.1867 -0.1104 0.1534>, rotation=<rot 90 0 0>}]
 dyn_hair_bk_r_base [{shapeType=Box, extentX=0.5746, extentY=0.5746, extentZ=0.5746}]
 dyn_hair_bk_r_parent [{shapeType=Box, extentX=4.8766, extentY=3.4314, extentZ=1.5239, center=<2.5612 -0.3205 -0.0171>, rotation=<rot -3.4435 -8.3485 -0.0001>}]
 dyn_hair_mid_l_01 [{shapeType=Capsule, length=3.7585, radius=1.3171, shapeName=dyn_hair_mid_l_01_capsule, center=<3 0 0>, rotation=<rot 90 0 0>}]
 dyn_hair_mid_l_parent [{shapeType=Box, extentX=5.3766, extentY=3.4314, extentZ=1.2208, center=<2.1949 0 0>}]
 head [{shapeType=Capsule, length=6, radius=6, shapeName=head_capsule, center=<3.5 3 0>, rotation=<rot 90 0 0>}]
 neck_01 [{shapeType=Capsule, length=7.3802, radius=5.1198, shapeName=neck_01_capsule, center=<3.6683 -0.2876 0>, rotation=<rot 90 0 0>}]
 spine_05 [
  {shapeType=Capsule, length=12, radius=9.25, shapeName=spine_05_capsule, center=<-2.5963 0.209 0>}
  {shapeType=Capsule, length=12, radius=7.75, shapeName=spine_05_capsule, center=<-6.4845 2.8778 0>}
 ]
}
constraints [
 {bone1Name=dyn_hair_mid_l_parent, bone2Name=head, swing1LimitDegrees=30, swing2LimitDegrees=15, twistLimitDegrees=3}
 {bone1Name=dyn_hair_bk_l_parent, bone2Name=head, swing1LimitDegrees=15, swing2LimitDegrees=30, twistLimitDegrees=2}
 {bone1Name=dyn_hair_mid_l_01, bone2Name=dyn_hair_mid_l_parent, swing1LimitDegrees=30, swing2LimitDegrees=60, twistLimitDegrees=3}
 {bone1Name=dyn_hair_bk_l_01, bone2Name=dyn_hair_bk_l_parent, swing1LimitDegrees=10, swing2LimitDegrees=30, twistLimitDegrees=3}
 {bone1Name=dyn_hair_bk_l_02, bone2Name=dyn_hair_bk_l_01, swing1LimitDegrees=10, swing2LimitDegrees=30, twistLimitDegrees=3}
 {bone1Name=dyn_hair_bk_mid_parent, bone2Name=head, swing1LimitDegrees=30, swing2LimitDegrees=27.96, twistLimitDegrees=1}
 {bone1Name=dyn_hair_bk_mid_01, bone2Name=dyn_hair_bk_mid_parent, swing1LimitDegrees=15, swing2LimitDegrees=30, twistLimitDegrees=3}
 {bone1Name=dyn_hair_bk_02, bone2Name=dyn_hair_bk_mid_01, swing1LimitDegrees=15, swing2LimitDegrees=15, twistLimitDegrees=0.5}
 {bone1Name=dyn_hair_bk_r_parent, bone2Name=head, swing1LimitDegrees=5, swing2LimitDegrees=15, twistLimitDegrees=0.5}
 {bone1Name=dyn_hair_bk_r_01, bone2Name=dyn_hair_bk_r_parent, swing1LimitDegrees=10, swing2LimitDegrees=15, twistLimitDegrees=0.5}
 {bone1Name=dyn_hair_bk_r_02, bone2Name=dyn_hair_bk_r_01, swing1LimitDegrees=10, swing2LimitDegrees=20, twistLimitDegrees=0.5}
 {
  bone1Name=clavicle_l
  bone2Name=spine_05
  swing1LimitDegrees=45
  swing1Motion=Free
  swing2LimitDegrees=45
  swing2Motion=Free
  twistLimitDegrees=45
  twistMotion=Free
 }
 {
  bone1Name=clavicle_r
  bone2Name=spine_05
  swing1LimitDegrees=45
  swing1Motion=Free
  swing2LimitDegrees=45
  swing2Motion=Free
  twistLimitDegrees=45
  twistMotion=Free
 }
]
