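# Physics asset for the front face-accessory part of the female medium Street Fashion Red player body, defining a jointed dyn_hair bone hierarchy.
# Capsule bodies form a chain from head through dyn_hair_parent to dyn_hair_03, branching to 04, 05 and 06, with 05 branching to 07 and 08.
# Every joint shares the same swing and tighter twist limits.
# Unconstrained capsules cover neck, upper spine and only the left clavicle, apparently as colliders.
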PhysicsAsset /Game/Characters/Player/Female/Medium/Bodies/F_MED_Street_Fashion_Red/Meshes/Parts/F_MED_Street_Fashion_Red_Frnt_FaceAcc_Physics
bodies {
 clavicle_l [{shapeType=Capsule, length=12, radius=5.5, shapeName=clavicle_l_capsule, center=<3.5 -1 -1>, rotation=<rot 80 -0.0003 -0.0003>}]
 dyn_hair_01 [{shapeType=Capsule, length=2.1476, radius=2, center=<3 -0.7268 0.055>, rotation=<rot 88.2827 179.9999 179.9999>}]
 dyn_hair_02 [{shapeType=Capsule, length=1.5, radius=3, center=<3.0508 -1.1464 -0.8769>, rotation=<rot 90 178.2101 158.21>}]
 dyn_hair_03 [{shapeType=Capsule, length=2.2124, radius=3.5, center=<3.1366 -0.3449 0.0001>, rotation=<rot 90 0 0>}]
 dyn_hair_04 [{shapeType=Capsule, length=1.8852, radius=3, center=<1.9598 0.2878 -1.1536>, rotation=<rot 80 0 -0.0003>}]
 dyn_hair_05 [{shapeType=Capsule, length=2.4833, radius=4, center=<4.5257 0.1059 -1.4375>, rotation=<rot 90 0 9.9999>}]
 dyn_hair_06 [{shapeType=Capsule, length=3, radius=1.3, center=<2.9058 0 0>, rotation=<rot 90 0 0>}]
 dyn_hair_07 [{shapeType=Capsule, length=2.3, radius=1.3, center=<2.5508 0 0>, rotation=<rot 90 0 0>}]
 dyn_hair_08 [{shapeType=Capsule, length=1.505, radius=1, center=<1.7662 0 0>, rotation=<rot 90 0 0>}]
 dyn_hair_parent [{shapeType=Capsule, length=3.6387, radius=2, center=<2.8154 0 0>, rotation=<rot 70 -0.0001 9.9997>}]
 head [{shapeType=Capsule, length=6, radius=6, shapeName=head_capsule, center=<3.5 3 0>, rotation=<rot 90 0 0>}]
 neck_01 [{shapeType=Capsule, length=8, radius=4.5, shapeName=neck_01_capsule, center=<3.6683 -0.2876 0>, rotation=<rot 90 0 0>}]
 spine_04 [{shapeType=Capsule, length=11.375, radius=18.625, center=<-6.5908 -6.9524 -0.9615>, rotation=<rot 89.3496 179.9999 180>}]
 spine_05 [
  {shapeType=Capsule, length=12, radius=9.25, shapeName=spine_05_capsule, center=<-2.5963 0.209 0>}
  {shapeType=Capsule, length=12, radius=7.75, shapeName=spine_05_capsule, center=<-6.4845 2.8778 0>}
 ]
}
constraints [
 {bone1Name=dyn_hair_02, bone2Name=dyn_hair_01, swing1LimitDegrees=25, swing2LimitDegrees=25, twistLimitDegrees=5}
 {bone1Name=dyn_hair_03, bone2Name=dyn_hair_02, swing1LimitDegrees=25, swing2LimitDegrees=25, twistLimitDegrees=5}
 {bone1Name=dyn_hair_04, bone2Name=dyn_hair_03, swing1LimitDegrees=25, swing2LimitDegrees=25, twistLimitDegrees=5}
 {bone1Name=dyn_hair_05, bone2Name=dyn_hair_03, swing1LimitDegrees=25, swing2LimitDegrees=25, twistLimitDegrees=5}
 {bone1Name=dyn_hair_06, bone2Name=dyn_hair_03, swing1LimitDegrees=25, swing2LimitDegrees=25, twistLimitDegrees=5}
 {bone1Name=dyn_hair_07, bone2Name=dyn_hair_05, swing1LimitDegrees=25, swing2LimitDegrees=25, twistLimitDegrees=5}
 {bone1Name=dyn_hair_08, bone2Name=dyn_hair_05, swing1LimitDegrees=25, swing2LimitDegrees=25, twistLimitDegrees=5}
 {bone1Name=dyn_hair_01, bone2Name=dyn_hair_parent, swing1LimitDegrees=25, swing2LimitDegrees=25, twistLimitDegrees=5}
 {bone1Name=dyn_hair_parent, bone2Name=head, swing1LimitDegrees=25, swing2LimitDegrees=25, twistLimitDegrees=5}
]
