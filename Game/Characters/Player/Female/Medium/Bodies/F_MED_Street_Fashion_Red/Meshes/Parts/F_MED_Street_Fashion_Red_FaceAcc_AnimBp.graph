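# Post-process style hair AnimBlueprint for the Street Fashion Red female FaceAcc mesh, following the attached parent's pose.
# Look At, constraint and RigidBody nodes shape the hair, with corrective offsets weighted by head, swing, skydiving and jump/fall alphas.
# Cached 'Pose' feeds front-end and in-game variants, blended by bIsInFrontEnd into 'FrontEndLogic'.
# bRigidBodyEnabled switches between that result and an AnimDynamics chain built on it, before final bone tweaks.
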AnimBlueprint /Game/Characters/Player/Female/Medium/Bodies/F_MED_Street_Fashion_Red/Meshes/Parts/F_MED_Street_Fashion_Red_FaceAcc_AnimBp
target_skeleton=&/F_MED_Street_Fashion_Red_FaceAcc_Skeleton
graphs [
 AnimGraph {nodes [
   n1 = "Transform (Modify) Bone - Bone: dyn_hair_02"(
    ComponentPose=Local To Component(LocalPose=Copy Pose From Mesh() {bUseAttachedParent=true})
    Alpha=1
    Translation="0,0,0"
    Rotation="6,12,0"
    Scale="1,1,1"
   ) {
    rotationMode=BMM_Additive
    rotationSpace=BCS_BoneSpace
    scaleMode=BMM_Additive
    scaleSpace=BCS_BoneSpace
    alphaBoolBlend {blendOption=Linear}
    boneToModify {boneName=dyn_hair_02}
    rotation {pitch=6, yaw=12}
   }
   n2 = "Transform (Modify) Bone - Bone: dyn_hair_03"(ComponentPose=@n1, Alpha=1, Translation="0,0,0", Rotation="0,10,0", Scale="1,1,1") {
    rotationMode=BMM_Additive
    scaleMode=BMM_Additive
    scaleSpace=BCS_BoneSpace
    alphaBoolBlend {blendOption=Linear}
    boneToModify {boneName=dyn_hair_03}
    rotation {yaw=10}
   }
   n3 = "Transform (Modify) Bone - Bone: dyn_hair_05"(ComponentPose=@n2, Alpha=1, Translation="0,0,0", Rotation="7,0,0", Scale="1,1,1") {
    rotationMode=BMM_Additive
    scaleMode=BMM_Additive
    scaleSpace=BCS_BoneSpace
    alphaBoolBlend {blendOption=Linear}
    boneToModify {boneName=dyn_hair_05}
    rotation {pitch=7}
   }
   Component To Local(ComponentPose=@n3)
   "Transform (Modify) Bone - Bone: dyn_hair_bk_r_01"!(ComponentPose, Alpha=1, Translation="0,-2,0", Rotation="0,-5,0") -> (Pose) {
    rotationMode=BMM_Additive
    rotationSpace=BCS_BoneSpace
    scaleMode=BMM_Additive
    scaleSpace=BCS_BoneSpace
    translationMode=BMM_Additive
    translationSpace=BCS_BoneSpace
    alphaBoolBlend {blendOption=Linear}
    boneToModify {boneName=dyn_hair_bk_r_01}
    rotation {yaw=-5}
    translation {y=-2}
   }
   "Transform (Modify) Bone - Bone: dyn_hair_bk_r_01"!(ComponentPose, Alpha=1, Translation="33.760,-64.372,63.192") -> (Pose) {
    rotationMode=BMM_Replace
    rotationSpace=BCS_BoneSpace
    scaleMode=BMM_Additive
    scaleSpace=BCS_BoneSpace
    translationMode=BMM_Additive
    translationSpace=BCS_BoneSpace
    alphaBoolBlend {blendOption=Linear}
    boneToModify {boneName=dyn_hair_bk_r_01}
    translation=<33.76 -64.372 63.192>
   }
   n4 = "Constraint - dyn_hair_02 (PP)"!(
    ComponentPose=Local To Component(LocalPose=Use cached pose 'FrontEndLogic'())
    Alpha=0.35
    ConstraintWeights_0=0.4
    ConstraintWeights_1:float=0.6
   ) -> (Pose) {
    alphaBoolBlend {blendOption=Linear}
    alphaScaleBiasClamp {bClampResult=true, clampMax=0.35}
    boneToModify {boneName=dyn_hair_02}
    constraintSetup [{targetBone {boneName=clavicle_l}}, {targetBone {boneName=neck_01}}]
    constraintWeights [1,1]
   }
   n5 = "Constraint - dyn_hair_bk_r_01 (PP)"!(ComponentPose=@n4, Alpha=0.35, ConstraintWeights_0=0.6, ConstraintWeights_1:float=0.4) -> (Pose) {
    alphaBoolBlend {blendOption=Linear}
    alphaScaleBiasClamp {bClampResult=true, clampMax=0.35}
    boneToModify {boneName=dyn_hair_bk_r_01}
    constraintSetup [{targetBone {boneName=spine_05}}, {targetBone {boneName=neck_02}}]
    constraintWeights [1,1]
   }
   n6 = "Constraint - dyn_hair_bk_l_01 (P)"(ComponentPose=@n5, Alpha=0.35, ConstraintWeights_0=1) {
    alphaBoolBlend {blendOption=Linear}
    alphaScaleBiasClamp {bClampResult=true, clampMax=0.35}
    boneToModify {boneName=dyn_hair_bk_l_01}
    constraintSetup [{targetBone {boneName=spine_05}}]
    constraintWeights [1]
   }
   n7 = "Constraint - dyn_hair_03 (PP)"!(ComponentPose=@n6, Alpha=0.35, ConstraintWeights_0=0.6, ConstraintWeights_1:float=0.4) -> (Pose) {
    alphaBoolBlend {blendOption=Linear}
    alphaScaleBiasClamp {bClampResult=true, clampMax=0.35}
    boneToModify {boneName=dyn_hair_03}
    constraintSetup [{targetBone {boneName=clavicle_l}}, {targetBone {boneName=spine_04}}]
    constraintWeights [1,1]
   }
   n8 = "Transform (Modify) Bone - Bone: dyn_hair_02"(ComponentPose=@n7, Alpha=Get Head Down() -> (Head Down:double), Translation="0,5,0", Rotation="0.277,-15.036,13.042", Scale="1,1,1") {
    rotationMode=BMM_Replace
    rotationSpace=BCS_BoneSpace
    scaleMode=BMM_Additive
    scaleSpace=BCS_BoneSpace
    translationMode=BMM_Additive
    translationSpace=BCS_BoneSpace
    alphaBoolBlend {blendOption=Linear}
    boneToModify {boneName=dyn_hair_02}
    rotation=<rot 0.277 -15.036 13.042>
    translation {y=5}
   }
   n9 = Anim Dynamics - Bone: dyn_hair_01(ComponentPose=@n8, Alpha=1) {
    angularDampingOverride=0.9999
    angularSpringConstant=500
    bAngularSpring=true
    bGravityOverrideInSimSpace=true
    bOverrideAngularDamping=true
    bUseGravityOverride=true
    simulationSpace=BoneRelative
    alphaBoolBlend {blendOption=Linear}
    boundBone {boneName=dyn_hair_01}
    componentAppliedLinearAccClamp=<5000 5000 5000>
    componentLinearAccScale=<1.2 1.2 1.2>
    componentLinearVelScale=<1.2 1.2 1.2>
    gravityOverride {z=-980}
    physicsBodyDefinitions [{
      sphereCollisionRadius=0
      boundBone {boneName=dyn_hair_01}
      boxExtents=<5 20 20>
      constraintSetup {angularLimitsMax=<1 5 10>, angularLimitsMin=<-1 -15 -5>, angularTarget=<0 0 0>}
      localJointOffset=<5 0 0>
     }]
    relativeSpaceBone {boneName=root}
   }
   n10 = Anim Dynamics - Bone: dyn_hair_02(ComponentPose=@n9, Alpha=1) {
    angularDampingOverride=0.9999
    angularSpringConstant=500
    bAngularSpring=true
    bGravityOverrideInSimSpace=true
    bOverrideAngularDamping=true
    bUseGravityOverride=true
    bUseSphericalLimits=true
    simulationSpace=BoneRelative
    alphaBoolBlend {blendOption=Linear}
    boundBone {boneName=dyn_hair_02}
    componentAppliedLinearAccClamp=<5000 5000 5000>
    componentLinearAccScale=<1.2 1.2 1.2>
    componentLinearVelScale=<1.2 1.2 1.2>
    gravityOverride {z=-980}
    physicsBodyDefinitions [{
      sphereCollisionRadius=0
      boundBone {boneName=dyn_hair_02}
      boxExtents=<5 20 20>
      constraintSetup {angularLimitsMax=<1 15 25>, angularLimitsMin=<-1 -20 -6>, angularTarget=<0 0 0>}
      localJointOffset=<5 0 0>
     }]
    relativeSpaceBone {boneName=root}
    sphericalLimits [{limitRadius=10, drivingBone {boneName=clavicle_l}, sphereLocalOffset=<0 0 0>}]
   }
   n11 = Anim Dynamics - Bone: dyn_hair_03(ComponentPose=@n10, Alpha=1, ExternalForce="0,0,0") {
    angularDampingOverride=0.9999
    angularSpringConstant=500
    bAngularSpring=true
    bGravityOverrideInSimSpace=true
    bOverrideAngularDamping=true
    bUseGravityOverride=true
    bUseSphericalLimits=true
    simulationSpace=BoneRelative
    alphaBoolBlend {blendOption=Linear}
    boundBone {boneName=dyn_hair_03}
    componentAppliedLinearAccClamp=<5000 5000 5000>
    componentLinearAccScale=<1.2 1.2 1.2>
    componentLinearVelScale=<1.2 1.2 1.2>
    gravityOverride {z=-980}
    physicsBodyDefinitions [{
      sphereCollisionRadius=0
      boundBone {boneName=dyn_hair_03}
      boxExtents=<5 20 20>
      constraintSetup {angularLimitsMax=<1 20 25>, angularLimitsMin=<-1 -45 -4>, angularTarget=<0 0 0>}
      localJointOffset=<5 0 0>
     }]
    relativeSpaceBone {boneName=root}
    sphericalLimits [{limitRadius=10, drivingBone {boneName=clavicle_l}, sphereLocalOffset=<0 0 0>}]
   }
   n12 = Anim Dynamics - Bone: dyn_hair_04(ComponentPose=@n11, Alpha=1, ExternalForce="0,0,0") {
    angularDampingOverride=0.9999
    angularSpringConstant=500
    bAngularSpring=true
    bGravityOverrideInSimSpace=true
    bOverrideAngularDamping=true
    bUseGravityOverride=true
    bUseSphericalLimits=true
    simulationSpace=BoneRelative
    alphaBoolBlend {blendOption=Linear}
    boundBone {boneName=dyn_hair_04}
    componentAppliedLinearAccClamp=<3500 3500 3500>
    componentLinearAccScale=<1.2 1.2 1.2>
    componentLinearVelScale=<1.2 1.2 1.2>
    gravityOverride {z=-980}
    physicsBodyDefinitions [{
      sphereCollisionRadius=0
      boundBone {boneName=dyn_hair_04}
      boxExtents=<5 20 20>
      constraintSetup {angularLimitsMax=<1 20 25>, angularLimitsMin=<-1 -45 -4>, angularTarget=<0 0 0>}
      localJointOffset=<5 0 0>
     }]
    relativeSpaceBone {boneName=root}
    sphericalLimits [{limitRadius=10, drivingBone {boneName=clavicle_l}, sphereLocalOffset=<0 0 0>}]
   }
   n13 = Anim Dynamics - Bone: dyn_hair_05(ComponentPose=@n12, Alpha=1, ExternalForce="0,0,0") {
    angularDampingOverride=0.9999
    angularSpringConstant=500
    bAngularSpring=true
    bGravityOverrideInSimSpace=true
    bOverrideAngularDamping=true
    bUseGravityOverride=true
    bUseSphericalLimits=true
    simulationSpace=BoneRelative
    alphaBoolBlend {blendOption=Linear}
    boundBone {boneName=dyn_hair_05}
    componentAppliedLinearAccClamp=<3500 3500 3500>
    componentLinearAccScale=<1.2 1.2 1.2>
    componentLinearVelScale=<1.2 1.2 1.2>
    gravityOverride {z=-980}
    physicsBodyDefinitions [{
      sphereCollisionRadius=0
      boundBone {boneName=dyn_hair_05}
      boxExtents=<5 20 20>
      constraintSetup {angularLimitsMax=<1 20 25>, angularLimitsMin=<-1 -45 -4>, angularTarget=<0 0 0>}
      localJointOffset=<5 0 0>
     }]
    relativeSpaceBone {boneName=root}
    sphericalLimits [{limitRadius=12, drivingBone {boneName=spine_03}, sphereLocalOffset=<10 2 -5>}]
   }
   n14 = Anim Dynamics - Bone: dyn_hair_06(ComponentPose=@n13, Alpha=1, ExternalForce="0,0,0") {
    angularDampingOverride=0.9999
    angularSpringConstant=500
    bAngularSpring=true
    bGravityOverrideInSimSpace=true
    bOverrideAngularDamping=true
    bUseGravityOverride=true
    simulationSpace=BoneRelative
    alphaBoolBlend {blendOption=Linear}
    boundBone {boneName=dyn_hair_06}
    componentAppliedLinearAccClamp=<5000 5000 5000>
    componentLinearAccScale=<1.2 1.2 1.2>
    componentLinearVelScale=<1.2 1.2 1.2>
    gravityOverride {z=-980}
    physicsBodyDefinitions [{
      sphereCollisionRadius=0
      boundBone {boneName=dyn_hair_06}
      boxExtents=<5 10 10>
      constraintSetup {angularLimitsMax=<1 5 15>, angularLimitsMin=<-1 -30 -10>, angularTarget=<0 0 0>}
      localJointOffset=<5 0 0>
     }]
    relativeSpaceBone {boneName=root}
    sphericalLimits [{limitRadius=10, drivingBone {boneName=spine_02}, sphereLocalOffset=<10 5 0>}]
   }
   n15 = Anim Dynamics - Bone: dyn_hair_07(ComponentPose=@n14, Alpha=1, ExternalForce="0,0,0") {
    angularDampingOverride=0.9999
    angularSpringConstant=500
    bAngularSpring=true
    bGravityOverrideInSimSpace=true
    bOverrideAngularDamping=true
    bUseGravityOverride=true
    simulationSpace=BoneRelative
    alphaBoolBlend {blendOption=Linear}
    boundBone {boneName=dyn_hair_07}
    componentAppliedLinearAccClamp=<5000 5000 5000>
    componentLinearAccScale=<1.2 1.2 1.2>
    componentLinearVelScale=<1.2 1.2 1.2>
    gravityOverride {z=-980}
    physicsBodyDefinitions [{
      sphereCollisionRadius=0
      boundBone {boneName=dyn_hair_07}
      boxExtents=<5 10 10>
      constraintSetup {angularLimitsMax=<1 5 15>, angularLimitsMin=<-1 -30 -10>, angularTarget=<0 0 0>}
      localJointOffset=<5 0 0>
     }]
    relativeSpaceBone {boneName=root}
    sphericalLimits [{limitRadius=12, drivingBone {boneName=spine_02}, sphereLocalOffset=<8 0 -5>}]
   }
   n16 = Anim Dynamics - Bone: dyn_hair_08(ComponentPose=@n15, Alpha=1, ExternalForce="0,0,0") {
    angularDampingOverride=0.9999
    angularSpringConstant=500
    bAngularSpring=true
    bGravityOverrideInSimSpace=true
    bOverrideAngularDamping=true
    bUseGravityOverride=true
    simulationSpace=BoneRelative
    alphaBoolBlend {blendOption=Linear}
    boundBone {boneName=dyn_hair_08}
    componentAppliedLinearAccClamp=<5000 5000 5000>
    componentLinearAccScale=<1.2 1.2 1.2>
    componentLinearVelScale=<1.2 1.2 1.2>
    gravityOverride {z=-980}
    physicsBodyDefinitions [{
      sphereCollisionRadius=0
      boundBone {boneName=dyn_hair_08}
      boxExtents=<5 10 10>
      constraintSetup {angularLimitsMax=<1 5 15>, angularLimitsMin=<-1 -30 -10>, angularTarget=<0 0 0>}
      localJointOffset=<5 0 0>
     }]
    relativeSpaceBone {boneName=root}
    sphericalLimits [{limitRadius=12, drivingBone {boneName=spine_02}, sphereLocalOffset=<5 0 -15>}]
   }
   n17 = Look At - Bone: dyn_hair_03!(ComponentPose=@n16, bAlphaBoolEnabled:bool=True) -> (Pose) {
    alphaInputType=Bool
    alphaBoolBlend {blendInTime=7, blendOption=Linear, blendOutTime=5}
    boneToModify {boneName=dyn_hair_03}
    lookAtLocation=<-16.9997 25.3056 -11.3486>
    lookAtTarget {boneReference {boneName=spine_05}}
    lookAt_Axis {axis=<1 0>}
   }
   n18 = Look At - Bone: dyn_hair_03(ComponentPose=@n17, bAlphaBoolEnabled=True) {
    alphaInputType=Bool
    alphaBoolBlend {blendInTime=7, blendOption=Linear, blendOutTime=5}
    boneToModify {boneName=dyn_hair_03}
    lookAtLocation=<-16.9997 25.3056 -11.3486>
    lookAtTarget {boneReference {boneName=spine_05}}
    lookAt_Axis {axis=<1 0>}
   }
   n19 = "Transform (Modify) Bone - Bone: dyn_hair_01"!(ComponentPose=@n18, Alpha=1, Translation="0,0,0", Rotation="-0.050,-26.557,0.073") -> (Pose) {
    rotationMode=BMM_Additive
    rotationSpace=BCS_BoneSpace
    scaleMode=BMM_Additive
    scaleSpace=BCS_BoneSpace
    translationMode=BMM_Additive
    translationSpace=BCS_BoneSpace
    alphaBoolBlend {blendOption=Linear}
    boneToModify {boneName=dyn_hair_01}
    rotation=<rot -0.05 -26.557 0.073>
   }
   n20 = "Transform (Modify) Bone - Bone: dyn_hair_03"!(ComponentPose=@n19, Alpha=1, Translation="0,0,0", Rotation="-0.106,22.496,0.075") -> (Pose) {
    rotationMode=BMM_Additive
    rotationSpace=BCS_BoneSpace
    scaleMode=BMM_Additive
    scaleSpace=BCS_BoneSpace
    translationMode=BMM_Additive
    translationSpace=BCS_BoneSpace
    alphaBoolBlend {blendOption=Linear}
    boneToModify {boneName=dyn_hair_03}
    rotation=<rot -0.106 22.496 0.075>
   }
   n21 = "Transform (Modify) Bone - Bone: dyn_hair_bk_r_parent"(ComponentPose=@n20, Alpha=1, Translation="0,0,0", Rotation="2.893,0.215,0.187", Scale="1,1,1") {
    rotationMode=BMM_Replace
    rotationSpace=BCS_BoneSpace
    scaleMode=BMM_Additive
    scaleSpace=BCS_BoneSpace
    translationMode=BMM_Additive
    translationSpace=BCS_BoneSpace
    alphaBoolBlend {blendOption=Linear}
    boneToModify {boneName=dyn_hair_bk_r_parent}
    rotation=<rot 2.893 0.215 0.187>
   }
   n22 = "Transform (Modify) Bone - Bone: dyn_hair_bk_mid_parent"(ComponentPose=@n21, Alpha=1, Translation="0,0,0", Rotation="54.203,39.362,34.629", Scale="1,1,1") {
    rotationMode=BMM_Replace
    rotationSpace=BCS_BoneSpace
    scaleMode=BMM_Additive
    scaleSpace=BCS_BoneSpace
    translationMode=BMM_Additive
    translationSpace=BCS_BoneSpace
    alphaBoolBlend {blendOption=Linear}
    boneToModify {boneName=dyn_hair_bk_mid_parent}
    rotation=<rot 54.203 39.362 34.629>
   }
   n23 = "Transform (Modify) Bone - Bone: dyn_hair_bk_l_parent"!(ComponentPose=@n22, Alpha=1, Translation="0,0,0", Rotation="24.170,-13.993,-15.457") -> (Pose) {
    rotationMode=BMM_Replace
    rotationSpace=BCS_BoneSpace
    scaleMode=BMM_Additive
    scaleSpace=BCS_BoneSpace
    translationMode=BMM_Replace
    translationSpace=BCS_BoneSpace
    alphaBoolBlend {blendOption=Linear}
    boneToModify {boneName=dyn_hair_bk_l_parent}
    rotation=<rot 24.17 -13.993 -15.457>
   }
   n24 = "Constraint - dyn_hair_bk_02 (P)"(ComponentPose=@n23, Alpha=1, ConstraintWeights_0=1) {
    alphaBoolBlend {blendOption=Linear}
    boneToModify {boneName=dyn_hair_bk_02}
    constraintSetup [{targetBone {boneName=spine_05}}]
    constraintWeights [1]
   }
   n25 = "Constraint - dyn_hair_bk_l_02 (P)"(ComponentPose=@n24, Alpha=1, ConstraintWeights_0=1) {
    alphaBoolBlend {blendOption=Linear}
    boneToModify {boneName=dyn_hair_bk_l_02}
    constraintSetup [{targetBone {boneName=spine_05}}]
    constraintWeights [1]
   }
   n26 = "Constraint - dyn_hair_bk_r_02 (P)"(ComponentPose=@n25, Alpha=1, ConstraintWeights_0=1) {
    alphaBoolBlend {blendOption=Linear}
    boneToModify {boneName=dyn_hair_bk_r_02}
    constraintSetup [{targetBone {boneName=spine_05}}]
    constraintWeights [1]
   }
   n27 = "Transform (Modify) Bone - Bone: dyn_hair_bk_l_parent"(ComponentPose=@n26, Alpha=1, Translation="0,2.2101,0", Rotation="-20.925,0.293,-0.202") {
    rotationMode=BMM_Replace
    rotationSpace=BCS_BoneSpace
    scaleMode=BMM_Additive
    scaleSpace=BCS_BoneSpace
    translationMode=BMM_Replace
    translationSpace=BCS_BoneSpace
    alphaBoolBlend {blendOption=Linear}
    boneToModify {boneName=dyn_hair_bk_l_parent}
    rotation=<rot -20.925 0.293 -0.202>
    translation {y=2.2101}
   }
   n28 = "Transform (Modify) Bone - Bone: dyn_hair_bk_mid_parent"!(ComponentPose=@n27, Alpha=1, Translation="0,0,0", Rotation="0,0,0") -> (Pose) {
    rotationMode=BMM_Replace
    rotationSpace=BCS_BoneSpace
    scaleMode=BMM_Additive
    scaleSpace=BCS_BoneSpace
    translationMode=BMM_Replace
    translationSpace=BCS_BoneSpace
    alphaBoolBlend {blendOption=Linear}
    boneToModify {boneName=dyn_hair_bk_mid_parent}
   }
   n29 = "Transform (Modify) Bone - Bone: dyn_hair_bk_r_01"(ComponentPose=@n28, Alpha=1, Translation="0,-3,0") {
    rotationMode=BMM_Replace
    rotationSpace=BCS_BoneSpace
    scaleMode=BMM_Additive
    scaleSpace=BCS_BoneSpace
    translationMode=BMM_Additive
    translationSpace=BCS_BoneSpace
    alphaBoolBlend {blendOption=Linear}
    alphaScaleBiasClamp {bInterpResult=true, interpSpeedDecreasing=5, interpSpeedIncreasing=7}
    boneToModify {boneName=dyn_hair_bk_r_01}
    rotation=<rot -0.007 30.9374 0.0075>
    translation {y=-3}
   }
   n30 = "Transform (Modify) Bone - Bone: dyn_hair_bk_l_01"!(ComponentPose=@n29, Alpha=1, Translation="0,-2,0", Rotation="-9.847,1.754,9.848") -> (Pose) {
    rotationMode=BMM_Additive
    rotationSpace=BCS_BoneSpace
    scaleMode=BMM_Additive
    scaleSpace=BCS_BoneSpace
    translationMode=BMM_Additive
    translationSpace=BCS_BoneSpace
    alphaBoolBlend {blendOption=Linear}
    boneToModify {boneName=dyn_hair_bk_l_01}
    rotation=<rot -9.847 1.754 9.8481>
    translation {y=-2}
   }
   n31 = "Transform (Modify) Bone - Bone: dyn_hair_01"(ComponentPose=@n30, Alpha=1, Translation="0,2,0", Rotation="0,0,0") {
    rotationMode=BMM_Replace
    rotationSpace=BCS_BoneSpace
    scaleMode=BMM_Additive
    scaleSpace=BCS_BoneSpace
    translationMode=BMM_Additive
    translationSpace=BCS_BoneSpace
    alphaBoolBlend {blendOption=Linear}
    boneToModify {boneName=dyn_hair_01}
    translation {y=2}
   }
   n32 = "Transform (Modify) Bone - Bone: dyn_hair_04"!(ComponentPose=@n31, Alpha=1, Translation="-0.3,2,0.75", Rotation="0,1,0") -> (Pose) {
    rotationMode=BMM_Replace
    rotationSpace=BCS_BoneSpace
    scaleMode=BMM_Additive
    scaleSpace=BCS_BoneSpace
    translationMode=BMM_Additive
    translationSpace=BCS_BoneSpace
    alphaBoolBlend {blendOption=Linear}
    boneToModify {boneName=dyn_hair_04}
    rotation {yaw=1}
    translation=<-0.3 2 0.75>
   }
   n33 = "Transform (Modify) Bone - Bone: dyn_hair_02"!(ComponentPose=@n32, Alpha=1, Translation="0,3,0", Rotation="0,2,0") -> (Pose) {
    rotationMode=BMM_Replace
    rotationSpace=BCS_BoneSpace
    scaleMode=BMM_Additive
    scaleSpace=BCS_BoneSpace
    translationMode=BMM_Replace
    translationSpace=BCS_BoneSpace
    alphaBoolBlend {blendOption=Linear}
    alphaScaleBiasClamp {bInterpResult=true, interpSpeedDecreasing=5, interpSpeedIncreasing=7}
    boneToModify {boneName=dyn_hair_02}
    rotation {yaw=2}
    translation {y=3}
   }
   n34 = "Transform (Modify) Bone - Bone: dyn_hair_bk_mid_base"!(ComponentPose=@n33, bAlphaBoolEnabled:bool=True, Translation="0,0,0", Rotation="4.629,-69.544,-6.307", Scale="1,1,1") -> (Pose) {
    alphaInputType=Bool
    rotationMode=BMM_Additive
    rotationSpace=BCS_BoneSpace
    translationMode=BMM_Additive
    translationSpace=BCS_BoneSpace
    alphaBoolBlend {blendOption=Linear}
    boneToModify {boneName=dyn_hair_bk_mid_base}
    rotation=<rot 4.629 -69.544 -6.3069>
   }
   n35 = "Transform (Modify) Bone - Bone: dyn_hair_bk_r_parent"!(ComponentPose=@n34, bAlphaBoolEnabled:bool=True, Translation="0,0,0", Rotation="4.306,-9.690,34.946", Scale="1,1,1") -> (Pose) {
    alphaInputType=Bool
    rotationMode=BMM_Additive
    rotationSpace=BCS_BoneSpace
    translationMode=BMM_Additive
    translationSpace=BCS_BoneSpace
    alphaBoolBlend {blendOption=Linear}
    boneToModify {boneName=dyn_hair_bk_r_parent}
    rotation=<rot 4.306 -9.69 34.946>
   }
   n36 = "Transform (Modify) Bone - Bone: dyn_hair_bk_mid_base"(ComponentPose=@n35, bAlphaBoolEnabled=True, Translation="0,0,0", Rotation="8.071,101.148,96.160", Scale="1,1,1") {
    alphaInputType=Bool
    rotationMode=BMM_Additive
    rotationSpace=BCS_BoneSpace
    translationMode=BMM_Additive
    translationSpace=BCS_BoneSpace
    alphaBoolBlend {blendOption=Linear}
    boneToModify {boneName=dyn_hair_bk_mid_base}
    rotation=<rot 8.071 101.148 96.16>
   }
   n37 = Component To Local(ComponentPose=@n36)
   n38 = Use cached pose 'FrontEndLogic'()
   n39 = Get bRigidBodyEnabled()
   n40 = Get JumpingFalling() -> (JumpingFalling:double)
   n41 = "Transform (Modify) Bone - Bone: dyn_hair_bk_r_01"!(
    ComponentPose=Local To Component(LocalPose=Blend Poses by bool(BlendPose_0=@n38, BlendPose_1=@n37, bActiveValue=@n39) {blendType=Linear})
    Alpha=@n40
    Translation="-0.103,0.373,0.589"
    Rotation:Rotator="10.000,0.000,0.000"
    Scale:Vector="0.500,0.750,0.750"
   ) -> (Pose) {
    alpha=0
    rotationMode=BMM_Additive
    rotationSpace=BCS_BoneSpace
    scaleMode=BMM_Additive
    scaleSpace=BCS_BoneSpace
    translationMode=BMM_Additive
    translationSpace=BCS_BoneSpace
    alphaBoolBlend {blendOption=Linear}
    alphaScaleBiasClamp {bInterpResult=true, interpSpeedDecreasing=8, interpSpeedIncreasing=8}
    boneToModify {boneName=dyn_hair_bk_r_01}
    rotation {pitch=10}
    scale=<0.5 0.75 0.75>
    translation=<-0.103 0.373 0.589>
   }
   n42 = "Transform (Modify) Bone - Bone: dyn_hair_bk_mid_01"(ComponentPose=@n41, Alpha=@n40, Translation="-0.148,1.360,0.577", Rotation="15.000,15.000,0.000", Scale="0.800,0.9,1.000") {
    alpha=0
    rotationMode=BMM_Additive
    rotationSpace=BCS_BoneSpace
    scaleMode=BMM_Additive
    scaleSpace=BCS_BoneSpace
    translationMode=BMM_Additive
    translationSpace=BCS_BoneSpace
    alphaBoolBlend {blendOption=Linear}
    alphaScaleBiasClamp {bInterpResult=true, interpSpeedDecreasing=8, interpSpeedIncreasing=8}
    boneToModify {boneName=dyn_hair_bk_mid_01}
    rotation=<rot 15 15.0001 0>
    scale=<0.8 0.9>
    translation=<-0.148 1.36 0.577>
   }
   n43 = "Transform (Modify) Bone - Bone: dyn_hair_bk_r_base"(
    ComponentPose=Local To Component(LocalPose=Component To Local(ComponentPose=@n42))
    Alpha=1
    Translation="-0.848,0,2"
    Rotation="0,0,0.5"
    Scale="1,1,1"
   ) {
    rotationMode=BMM_Additive
    rotationSpace=BCS_BoneSpace
    scaleMode=BMM_Additive
    scaleSpace=BCS_BoneSpace
    translationMode=BMM_Additive
    translationSpace=BCS_BoneSpace
    alphaBoolBlend {blendOption=Linear}
    alphaScaleBiasClamp {bInterpResult=true, interpSpeedIncreasing=5}
    boneToModify {boneName=dyn_hair_bk_r_base}
    rotation {roll=0.5}
    translation {x=-0.848, z=2}
   }
   n44 = "Transform (Modify) Bone - Bone: dyn_hair_bk_r_01"(ComponentPose=@n43, Alpha=1, Translation="-0.25,-0.565,0", Rotation="0.069,0.085,43.906", Scale="0.83,0.83,1") {
    rotationMode=BMM_Additive
    rotationSpace=BCS_BoneSpace
    scaleMode=BMM_Additive
    scaleSpace=BCS_BoneSpace
    translationMode=BMM_Additive
    translationSpace=BCS_BoneSpace
    alphaBoolBlend {blendOption=Linear}
    alphaScaleBiasClamp {bInterpResult=true, interpSpeedIncreasing=5}
    boneToModify {boneName=dyn_hair_bk_r_01}
    rotation=<rot 0.069 0.085 43.906>
    scale=<0.83 0.83>
    translation=<-0.25 -0.565>
   }
   n45 = "Transform (Modify) Bone - Bone: dyn_hair_bk_r_02"(ComponentPose=@n44, Alpha=1, Translation="-0.25,0,0", Rotation="0,0,0", Scale="0.83,0.83,1") {
    rotationMode=BMM_Additive
    rotationSpace=BCS_BoneSpace
    scaleMode=BMM_Additive
    scaleSpace=BCS_BoneSpace
    translationMode=BMM_Additive
    translationSpace=BCS_BoneSpace
    alphaBoolBlend {blendOption=Linear}
    alphaScaleBiasClamp {bInterpResult=true, interpSpeedIncreasing=5}
    boneToModify {boneName=dyn_hair_bk_r_02}
    scale=<0.83 0.83>
    translation {x=-0.25}
   }
   n46 = "Transform (Modify) Bone - Bone: dyn_hair_bk_mid_parent"!(ComponentPose=@n45, Alpha=1, Translation="0,0,0", Rotation="0,0,0.5", Scale:Vector="1,1,1") -> (Pose) {
    rotationMode=BMM_Additive
    rotationSpace=BCS_BoneSpace
    scaleMode=BMM_Additive
    scaleSpace=BCS_BoneSpace
    translationMode=BMM_Additive
    translationSpace=BCS_BoneSpace
    alphaBoolBlend {blendOption=Linear}
    alphaScaleBiasClamp {bInterpResult=true, interpSpeedIncreasing=5}
    boneToModify {boneName=dyn_hair_bk_mid_parent}
    rotation {roll=0.5}
   }
   Output Pose(Result=Component To Local(ComponentPose=@n46)) {layerGroup=None}
   n47 = Get HeadRotBackAlpha() -> (HeadRotBackAlpha:float)
   n48 = Get HeadRotBackAlpha()
   n49 = "Transform (Modify) Bone - Bone: dyn_hair_bk_l_01"!(
    ComponentPose=Local To Component(LocalPose=Use cached pose 'Pose'())
    Alpha=@n48
    Translation="0,0,0"
    Rotation="0,15,0"
    Scale:Vector="0.8,0.8,0.8"
   ) -> (Pose) {
    rotationMode=BMM_Additive
    rotationSpace=BCS_BoneSpace
    scaleMode=BMM_Additive
    scaleSpace=BCS_BoneSpace
    alphaBoolBlend {blendOption=Linear}
    boneToModify {boneName=dyn_hair_bk_l_01}
    rotation {yaw=15}
    scale=<0.8 0.8 0.8>
   }
   n50 = "Transform (Modify) Bone - Bone: dyn_hair_bk_mid_01"(ComponentPose=@n49, Alpha=@n48, Translation="0,0,0", Rotation="0,10,0", Scale="0.800,0.8,0.800") {
    rotationMode=BMM_Additive
    rotationSpace=BCS_BoneSpace
    scaleMode=BMM_Additive
    scaleSpace=BCS_BoneSpace
    alphaBoolBlend {blendOption=Linear}
    boneToModify {boneName=dyn_hair_bk_mid_01}
    rotation {yaw=10}
    scale=<0.8 0.8 0.8>
   }
   n51 = Get HeadRotBackAlpha()
   n52 = "Transform (Modify) Bone - Bone: dyn_hair_bk_r_02"(ComponentPose=@n50, Alpha=@n51, Translation="0,0,0", Rotation="0,0,0", Scale="0.7,0.700,0.7") {
    rotationMode=BMM_Additive
    rotationSpace=BCS_BoneSpace
    scaleMode=BMM_Additive
    scaleSpace=BCS_BoneSpace
    translationMode=BMM_Additive
    translationSpace=BCS_BoneSpace
    alphaBoolBlend {blendOption=Linear}
    boneToModify {boneName=dyn_hair_bk_r_02}
    scale=<0.7 0.7 0.7>
   }
   n53 = "Transform (Modify) Bone - Bone: dyn_hair_bk_r_01"(ComponentPose=@n52, Alpha=@n51, Translation="0,0,0", Rotation="0,0,0", Scale="0.2,0.15,1.000") {
    rotationMode=BMM_Additive
    rotationSpace=BCS_BoneSpace
    scaleMode=BMM_Additive
    scaleSpace=BCS_BoneSpace
    translationMode=BMM_Additive
    translationSpace=BCS_BoneSpace
    alphaBoolBlend {blendOption=Linear}
    boneToModify {boneName=dyn_hair_bk_r_01}
    scale=<0.2 0.15>
   }
   n54 = "Transform (Modify) Bone - Bone: dyn_hair_bk_r_parent"!(ComponentPose=@n53, Alpha:float=@n47, Translation="0,0,0", Rotation="0,0,0", Scale="0.7,1,1.000") -> (Pose) {
    rotationMode=BMM_Additive
    rotationSpace=BCS_BoneSpace
    scaleMode=BMM_Additive
    scaleSpace=BCS_BoneSpace
    translationMode=BMM_Additive
    translationSpace=BCS_BoneSpace
    alphaBoolBlend {blendOption=Linear}
    alphaScaleBiasClamp {interpSpeedDecreasing=30, interpSpeedIncreasing=30}
    boneToModify {boneName=dyn_hair_bk_r_parent}
    scale {x=0.7}
   }
   n55 = "Transform (Modify) Bone - Bone: dyn_hair_bk_mid_parent"(ComponentPose=@n54, Alpha=@n47, Translation="0,0,0", Rotation="0,0,0", Scale="0.75,1,1.000") {
    rotationMode=BMM_Additive
    rotationSpace=BCS_BoneSpace
    scaleMode=BMM_Additive
    scaleSpace=BCS_BoneSpace
    translationSpace=BCS_BoneSpace
    alphaBoolBlend {blendOption=Linear}
    alphaScaleBiasClamp {interpSpeedDecreasing=30, interpSpeedIncreasing=30}
    boneToModify {boneName=dyn_hair_bk_mid_parent}
    scale {x=0.75}
   }
   n56 = "Transform (Modify) Bone - Bone: dyn_hair_bk_l_parent"!(ComponentPose=@n55, Alpha=@n47, Translation="0,0,0", Rotation="0,0,0", Scale:Vector="0.8,1.000,1.000") -> (Pose) {
    rotationMode=BMM_Additive
    rotationSpace=BCS_BoneSpace
    scaleMode=BMM_Additive
    scaleSpace=BCS_BoneSpace
    translationSpace=BCS_BoneSpace
    alphaBoolBlend {blendOption=Linear}
    alphaScaleBiasClamp {interpSpeedDecreasing=30, interpSpeedIncreasing=30}
    boneToModify {boneName=dyn_hair_bk_l_parent}
    scale {x=0.8}
   }
   n57 = "Constraint - dyn_hair_bk_r_01 (PPPP)"!(
    ComponentPose=@n56
    Alpha=@n47
    ConstraintWeights_0=0.0433
    ConstraintWeights_1:float=0.0416
    ConstraintWeights_2:float=0.4958
    ConstraintWeights_3:float=0.4193
   ) -> (Pose) {
    alphaBoolBlend {blendOption=Linear}
    boneToModify {boneName=dyn_hair_bk_r_01}
    constraintSetup [
     {targetBone {boneName=head}}
     {targetBone {boneName=dyn_hair_bk_r_base}}
     {targetBone {boneName=dyn_hair_bk_r_parent}}
     {targetBone {boneName=dyn_hair_bk_r_01}}
    ]
    constraintWeights [1,1,1,1]
   }
   n58 = "Constraint - dyn_hair_bk_r_02 (PP)"!(ComponentPose=@n57, Alpha=@n47, ConstraintWeights_0=0.6216, ConstraintWeights_1:float=0.3784) -> (Pose) {
    alphaBoolBlend {blendOption=Linear}
    boneToModify {boneName=dyn_hair_bk_r_02}
    constraintSetup [{targetBone {boneName=dyn_hair_bk_r_01}}, {targetBone {boneName=dyn_hair_bk_r_02}}]
    constraintWeights [1,1]
   }
   n59 = "Constraint - dyn_hair_03 (P)"(ComponentPose=@n58, Alpha=1, ConstraintWeights_0=1) {
    alphaBoolBlend {blendOption=Linear}
    alphaScaleBiasClamp {bClampResult=true, bInterpResult=true, clampMax=0.7, interpSpeedDecreasing=5, interpSpeedIncreasing=7}
    boneToModify {boneName=dyn_hair_03}
    constraintSetup [{targetBone {boneName=spine_05}}]
    constraintWeights [1]
   }
   n60 = "Transform (Modify) Bone - Bone: dyn_hair_05"!(ComponentPose=@n59, Alpha=1, Translation="0,0,0", Rotation="-1.610,40.289,9.935") -> (Pose) {
    rotationMode=BMM_Additive
    rotationSpace=BCS_BoneSpace
    scaleSpace=BCS_BoneSpace
    translationMode=BMM_Additive
    translationSpace=BCS_BoneSpace
    alphaBoolBlend {blendOption=Linear}
    alphaScaleBiasClamp {bInterpResult=true, interpSpeedDecreasing=5, interpSpeedIncreasing=7}
    boneToModify {boneName=dyn_hair_05}
    rotation=<rot -1.61 40.289 9.935>
    scale=<0.2 0.15>
   }
   n61 = "Constraint - dyn_hair_bk_r_01 (PP)"(ComponentPose=@n60, Alpha=0.35, ConstraintWeights_0=0.4, ConstraintWeights_1=0.6) {
    alphaBoolBlend {blendOption=Linear}
    alphaScaleBiasClamp {bClampResult=true, bInterpResult=true, clampMax=0.8, interpSpeedIncreasing=1}
    boneToModify {boneName=dyn_hair_bk_r_01}
    constraintSetup [{targetBone {boneName=clavicle_r}}, {targetBone {boneName=neck_01}}]
    constraintWeights [1,1]
   }
   n62 = "Constraint - dyn_hair_bk_r_02 (P)"(ComponentPose=@n61, Alpha=0.35, ConstraintWeights_0=1) {
    alphaBoolBlend {blendOption=Linear}
    alphaScaleBiasClamp {bClampResult=true, bInterpResult=true, interpSpeedIncreasing=1}
    boneToModify {boneName=dyn_hair_bk_r_02}
    constraintSetup [{targetBone {boneName=spine_05}}]
    constraintWeights [1]
   }
   n63 = "Constraint - dyn_hair_bk_mid_01 (P)"(ComponentPose=@n62, Alpha=0.35, ConstraintWeights_0=1) {
    alphaBoolBlend {blendOption=Linear}
    alphaScaleBiasClamp {bClampResult=true, bInterpResult=true, interpSpeedIncreasing=1}
    boneToModify {boneName=dyn_hair_bk_mid_01}
    constraintSetup [{targetBone {boneName=spine_05}}]
    constraintWeights [1]
   }
   n64 = "Constraint - dyn_hair_bk_l_01 (P)"(ComponentPose=@n63, Alpha=0.35, ConstraintWeights_0=1) {
    alphaBoolBlend {blendOption=Linear}
    alphaScaleBiasClamp {bClampResult=true, bInterpResult=true, interpSpeedIncreasing=1}
    boneToModify {boneName=dyn_hair_bk_l_01}
    constraintSetup [{targetBone {boneName=spine_05}}]
    constraintWeights [1]
   }
   n65 = "Constraint - dyn_hair_01 (P)"(ComponentPose=Local To Component(LocalPose=Use cached pose 'Pose'()), Alpha=0.35, ConstraintWeights_0=1) {
    alphaBoolBlend {blendOption=Linear}
    alphaScaleBiasClamp {bClampResult=true, clampMax=0.35}
    boneToModify {boneName=dyn_hair_01}
    constraintSetup [{targetBone {boneName=clavicle_l}}]
    constraintWeights [1]
   }
   n66 = "Constraint - dyn_hair_bk_r_02 (P)"(ComponentPose=@n65, Alpha=0.35, ConstraintWeights_0=1) {
    alphaBoolBlend {blendOption=Linear}
    alphaScaleBiasClamp {bClampResult=true, clampMax=0.8}
    boneToModify {boneName=dyn_hair_bk_r_02}
    constraintSetup [{targetBone {boneName=clavicle_r}}]
    constraintWeights [1]
   }
   n67 = "Constraint - dyn_hair_bk_mid_01 (P)"(ComponentPose=@n66, Alpha=0.35, ConstraintWeights_0=1) {
    alphaBoolBlend {blendOption=Linear}
    alphaScaleBiasClamp {bClampResult=true, clampMax=0.35}
    boneToModify {boneName=dyn_hair_bk_mid_01}
    constraintSetup [{targetBone {boneName=clavicle_r}}]
    constraintWeights [1]
   }
   n68 = "Constraint - dyn_hair_bk_r_01 (PP)"(ComponentPose=@n67, Alpha=0.35, ConstraintWeights_0=0.4, ConstraintWeights_1=0.6) {
    alphaBoolBlend {blendOption=Linear}
    alphaScaleBiasClamp {bClampResult=true}
    boneToModify {boneName=dyn_hair_bk_r_01}
    constraintSetup [{targetBone {boneName=clavicle_r}}, {targetBone {boneName=neck_01}}]
    constraintWeights [1,1]
   }
   n69 = "Constraint - dyn_hair_bk_l_01 (P)"(ComponentPose=@n68, Alpha=0.35, ConstraintWeights_0=1) {
    alphaBoolBlend {blendOption=Linear}
    alphaScaleBiasClamp {bClampResult=true, bInterpResult=true, interpSpeedIncreasing=1}
    boneToModify {boneName=dyn_hair_bk_l_01}
    constraintSetup [{targetBone {boneName=spine_05}}]
    constraintWeights [1]
   }
   n70 = "Transform (Modify) Bone - Bone: dyn_hair_01"(ComponentPose=@n69, Alpha=1, Translation="2.635,1.418,1.635", Rotation="9,0,-15") {
    rotationMode=BMM_Additive
    rotationSpace=BCS_BoneSpace
    scaleMode=BMM_Additive
    scaleSpace=BCS_BoneSpace
    translationMode=BMM_Additive
    translationSpace=BCS_BoneSpace
    alphaBoolBlend {blendOption=Linear}
    boneToModify {boneName=dyn_hair_01}
    rotation {pitch=9, roll=-15}
    translation=<2.635 1.418 1.635>
   }
   n71 = "Transform (Modify) Bone - Bone: dyn_hair_04"!(ComponentPose=@n70, Alpha=1, Translation="1.498,0.938,-0.213") -> (Pose) {
    rotationMode=BMM_Additive
    rotationSpace=BCS_BoneSpace
    scaleMode=BMM_Additive
    scaleSpace=BCS_BoneSpace
    translationMode=BMM_Additive
    translationSpace=BCS_BoneSpace
    alphaBoolBlend {blendOption=Linear}
    boneToModify {boneName=dyn_hair_04}
    translation=<1.498 0.938 -0.213>
   }
   n72 = RigidBody!(ComponentPose=@n71, bAlphaBoolEnabled:bool, ExternalForce) -> (Pose) {
    alphaInputType=Bool
    overridePhysicsAsset=&/F_MED_Street_Fashion_Red_Frnt_FaceAcc_Physics
    alphaBoolBlend {blendOption=Linear}
   }
   Save cached pose 'FrontEndLogic'(Pose=Blend Poses by bool(BlendPose_0=Component To Local(ComponentPose=@n72), BlendPose_1=Component To Local(ComponentPose=@n64), bActiveValue=Get bIsInFrontEnd()) {blendType=Linear}) {cacheName=FrontEndLogic}
   n73 = Get HeadRotBackAlpha()
   n74 = Look At - Bone: dyn_hair_bk_l_base(ComponentPose=Local To Component(LocalPose=Copy Pose From Mesh() {bUseAttachedParent=true}), Alpha=@n73) {
    lookAtClamp=100
    alphaBoolBlend {blendOption=Linear}
    boneToModify {boneName=dyn_hair_bk_l_base}
    lookAtLocation=<1 1 5>
    lookAtTarget {boneReference {boneName=spine_04}}
    lookAt_Axis {axis=<1 0>}
    lookUp_Axis {axis {x=0, z=1}}
   }
   n75 = Look At - Bone: dyn_hair_bk_l_01(ComponentPose=@n74, Alpha=@n73) {
    lookAtClamp=100
    alphaBoolBlend {blendOption=Linear}
    boneToModify {boneName=dyn_hair_bk_l_01}
    lookAtLocation=<1 1 5>
    lookAtTarget {boneReference {boneName=spine_04}}
    lookAt_Axis {axis=<1 0>}
    lookUp_Axis {axis {x=0, z=1}}
   }
   n76 = Look At - Bone: dyn_hair_bk_l_02(ComponentPose=@n75, Alpha=@n73) {
    lookAtClamp=100
    alphaBoolBlend {blendOption=Linear}
    boneToModify {boneName=dyn_hair_bk_l_02}
    lookAtLocation=<1 1 5>
    lookAtTarget {boneReference {boneName=spine_03}}
    lookAt_Axis {axis=<1 0>}
    lookUp_Axis {axis {x=0, z=1}}
   }
   n77 = Look At - Bone: dyn_hair_bk_mid_base(ComponentPose=@n76, Alpha=@n73) {
    lookAtClamp=100
    alphaBoolBlend {blendOption=Linear}
    boneToModify {boneName=dyn_hair_bk_mid_base}
    lookAtLocation=<1 1 5>
    lookAtTarget {boneReference {boneName=spine_04}}
    lookAt_Axis {axis=<1 0>}
    lookUp_Axis {axis {x=0, z=1}}
   }
   n78 = Look At - Bone: dyn_hair_bk_mid_01(ComponentPose=@n77, Alpha=@n73) {
    lookAtClamp=100
    alphaBoolBlend {blendOption=Linear}
    boneToModify {boneName=dyn_hair_bk_mid_01}
    lookAtLocation=<1 1 5>
    lookAtTarget {boneReference {boneName=spine_04}}
    lookAt_Axis {axis=<1 0>}
    lookUp_Axis {axis {x=0, z=1}}
   }
   n79 = Look At - Bone: dyn_hair_bk_02(ComponentPose=@n78, Alpha=@n73) {
    lookAtClamp=100
    alphaBoolBlend {blendOption=Linear}
    boneToModify {boneName=dyn_hair_bk_02}
    lookAtLocation=<1 1 5>
    lookAtTarget {boneReference {boneName=spine_05}}
    lookAt_Axis {axis=<1 0>}
    lookUp_Axis {axis {x=0, z=1}}
   }
   n80 = Look At - Bone: dyn_hair_bk_r_02(ComponentPose=@n79, Alpha=@n73) {
    lookAtClamp=100
    alphaBoolBlend {blendOption=Linear}
    boneToModify {boneName=dyn_hair_bk_r_02}
    lookAtLocation=<1 1 5>
    lookAtTarget {boneReference {boneName=spine_05}}
    lookAt_Axis {axis=<1 0>}
    lookUp_Axis {axis {x=0, z=1}}
   }
   n81 = "Constraint - dyn_hair_bk_l_01 (P)"(ComponentPose=@n80, Alpha=1, ConstraintWeights_0=1) {
    alphaBoolBlend {blendOption=Linear}
    boneToModify {boneName=dyn_hair_bk_l_01}
    constraintSetup [{targetBone {boneName=dyn_hair_bk_mid_01}}]
    constraintWeights [1]
   }
   n82 = "Constraint - dyn_hair_bk_l_02 (P)"(ComponentPose=@n81, Alpha=1, ConstraintWeights_0=1) {
    alphaBoolBlend {blendOption=Linear}
    boneToModify {boneName=dyn_hair_bk_l_02}
    constraintSetup [{targetBone {boneName=dyn_hair_bk_02}}]
    constraintWeights [1]
   }
   n83 = "Constraint - dyn_hair_bk_r_parent (P)"(ComponentPose=@n82, Alpha=1, ConstraintWeights_0=1) {
    alphaBoolBlend {blendOption=Linear}
    boneToModify {boneName=dyn_hair_bk_r_parent}
    constraintSetup [{targetBone {boneName=neck_01}}]
    constraintWeights [1]
   }
   n84 = "Constraint - dyn_hair_bk_r_01 (P)"(ComponentPose=@n83, Alpha=0.7, ConstraintWeights_0=1) {
    alphaBoolBlend {blendOption=Linear}
    boneToModify {boneName=dyn_hair_bk_r_01}
    constraintSetup [{targetBone {boneName=upperarm_r}}]
    constraintWeights [1]
   }
   n85 = "Constraint - dyn_hair_bk_r_02 (P)"(ComponentPose=@n84, Alpha=1, ConstraintWeights_0=1) {
    alphaBoolBlend {blendOption=Linear}
    boneToModify {boneName=dyn_hair_bk_r_02}
    constraintSetup [{targetBone {boneName=dyn_hair_bk_02}}]
    constraintWeights [1]
   }
   n86 = "Constraint - dyn_hair_bk_mid_01 (P)"(ComponentPose=@n85, Alpha=1, ConstraintWeights_0=1) {
    alphaBoolBlend {blendOption=Linear}
    boneToModify {boneName=dyn_hair_bk_mid_01}
    constraintSetup [{targetBone {boneName=dyn_hair_bk_r_01}}]
    constraintWeights [1]
   }
   n87 = "Constraint - dyn_hair_bk_02 (P)"(ComponentPose=@n86, Alpha=1, ConstraintWeights_0=1) {
    alphaBoolBlend {blendOption=Linear}
    boneToModify {boneName=dyn_hair_bk_02}
    constraintSetup [{targetBone {boneName=dyn_hair_bk_l_02}}]
    constraintWeights [1]
   }
   n88 = Look At - Bone: dyn_hair_base(ComponentPose=@n87, Alpha=Get Swing() -> (Swing:double)) {
    lookAtClamp=100
    alphaBoolBlend {blendOption=Linear}
    boneToModify {boneName=dyn_hair_base}
    lookAtLocation=<1 1 5>
    lookAtTarget {boneReference {boneName=lowerarm_l}}
    lookAt_Axis {axis=<1 0>}
    lookUp_Axis {axis {z=1}}
   }
   n89 = RigidBody!(ComponentPose=@n88, Alpha:float=Get SkydivingBackAlpha() -> (SkydivingBackAlpha:float), ExternalForce) -> (Pose) {
    bClampLinearTranslationLimitToRefPose=true
    bOverrideWorldGravity=true
    overridePhysicsAsset=&/F_MED_Street_Fashion_Red_Back_FaceAcc_Physics
    simulationSpace=BaseBoneSpace
    alphaBoolBlend {blendOption=Linear}
    baseBoneRef {boneName=spine_05}
    componentAppliedLinearAccClamp=<1000 1000 1000>
    componentLinearAccScale=<0.25 0.25 0.25>
    componentLinearVelScale=<0.25 0.25 0.25>
    overrideWorldGravity {z=-980}
   }
   n90 = RigidBody(ComponentPose=@n89, Alpha=Get SkydivingFrontAlpha() -> (SkydivingFrontAlpha:float)) {
    bClampLinearTranslationLimitToRefPose=true
    bOverrideWorldGravity=true
    overridePhysicsAsset=&F_MED_Street_Fashion_Red_Frnt_FaceAcc_Physics
    simulationSpace=BaseBoneSpace
    alphaBoolBlend {blendOption=Linear}
    baseBoneRef {boneName=neck_01}
    componentAppliedLinearAccClamp=<1000 1000 1000>
    componentLinearAccScale=<0.5 0.5 0.5>
    componentLinearVelScale=<0.5 0.5 0.5>
    overrideWorldGravity {z=-980}
   }
   n91 = RigidBody!(ComponentPose=@n90, bAlphaBoolEnabled:bool, ExternalForce) -> (Pose) {
    alphaInputType=Bool
    bClampLinearTranslationLimitToRefPose=true
    bOverrideWorldGravity=true
    overridePhysicsAsset=&F_MED_Street_Fashion_Red_Back_FaceAcc_Physics
    simulationSpace=BaseBoneSpace
    alphaBoolBlend {blendOption=Linear}
    baseBoneRef {boneName=neck_01}
    componentAppliedLinearAccClamp=<1000 1000 1000>
    componentLinearAccScale=<0.5 0.5 0.5>
    componentLinearVelScale=<0.5 0.5 0.5>
    overrideWorldGravity {z=-980}
   }
   n92 = "Transform (Modify) Bone - Bone: dyn_hair_mid_l_base"!(ComponentPose=@n91, bAlphaBoolEnabled:bool=True, Translation="0,0,0", Rotation="10,0,0", Scale="1,1,1") -> (Pose) {
    alphaInputType=Bool
    rotationMode=BMM_Additive
    rotationSpace=BCS_BoneSpace
    translationMode=BMM_Additive
    translationSpace=BCS_BoneSpace
    alphaBoolBlend {blendOption=Linear}
    boneToModify {boneName=dyn_hair_mid_l_base}
    rotation {pitch=10}
   }
   n93 = "Transform (Modify) Bone - Bone: dyn_hair_bk_l_base"!(ComponentPose=@n92, bAlphaBoolEnabled:bool=True, Translation="0,0,0", Rotation="10,0,0", Scale="1,1,1") -> (Pose) {
    alphaInputType=Bool
    rotationMode=BMM_Additive
    rotationSpace=BCS_BoneSpace
    translationMode=BMM_Additive
    translationSpace=BCS_BoneSpace
    alphaBoolBlend {blendOption=Linear}
    boneToModify {boneName=dyn_hair_bk_l_base}
    rotation {pitch=10}
   }
   n94 = "Constraint - dyn_hair_mid_l_01 (P)"!(ComponentPose=@n93, bAlphaBoolEnabled:bool=True, ConstraintWeights_0=1) -> (Pose) {
    alphaInputType=Bool
    alphaBoolBlend {blendOption=Linear}
    boneToModify {boneName=dyn_hair_mid_l_01}
    constraintSetup [{targetBone {boneName=upperarm_l}}]
    constraintWeights [1]
   }
   n95 = "Constraint - dyn_hair_bk_02 (P)"!(ComponentPose=@n94, bAlphaBoolEnabled:bool=True, ConstraintWeights_0=1) -> (Pose) {
    alphaInputType=Bool
    alphaBoolBlend {blendOption=Linear}
    boneToModify {boneName=dyn_hair_bk_02}
    constraintSetup [{targetBone {boneName=upperarm_l}}]
    constraintWeights [1]
   }
   n96 = "Transform (Modify) Bone - Bone: dyn_hair_bk_mid_base"(ComponentPose=@n95, bAlphaBoolEnabled=True, Translation="0,0,1", Rotation="20,0,0", Scale="1,1,1") {
    alphaInputType=Bool
    rotationMode=BMM_Additive
    rotationSpace=BCS_BoneSpace
    translationMode=BMM_Additive
    translationSpace=BCS_BoneSpace
    alphaBoolBlend {blendOption=Linear}
    boneToModify {boneName=dyn_hair_bk_mid_base}
    rotation {pitch=20}
    translation {z=1}
   }
   n97 = "Transform (Modify) Bone - Bone: dyn_hair_bk_r_base"!(ComponentPose=@n96, bAlphaBoolEnabled:bool=True, Translation="0,0,1", Rotation="20,0,0", Scale="1,1,1") -> (Pose) {
    alphaInputType=Bool
    rotationMode=BMM_Additive
    rotationSpace=BCS_BoneSpace
    translationMode=BMM_Additive
    translationSpace=BCS_BoneSpace
    alphaBoolBlend {blendOption=Linear}
    boneToModify {boneName=dyn_hair_bk_r_base}
    rotation {pitch=20}
    translation {z=1}
   }
   n98 = "Transform (Modify) Bone - Bone: head"!(ComponentPose=@n97, bAlphaBoolEnabled:bool=True, Translation="1,-0.5,0", Rotation="0,5,0", Scale="1,1,1") -> (Pose) {
    alphaInputType=Bool
    rotationMode=BMM_Additive
    rotationSpace=BCS_BoneSpace
    translationMode=BMM_Additive
    translationSpace=BCS_BoneSpace
    alphaBoolBlend {blendOption=Linear}
    boneToModify {boneName=head}
    rotation {yaw=5}
    translation=<1 -0.5>
   }
   Save cached pose 'Pose'(Pose=Component To Local(ComponentPose=@n98)) {cacheName=Pose}
   Reroute Node(InputPin:Wildcard) -> (OutputPin:Wildcard)
   Get HeadSwingAlpha() -> (HeadSwingAlpha:float)
  ]}
 EventGraph {nodes [
   Event BlueprintInitializeAnimation()
   > Set SkelMeshOwn(SkelMeshOwn:Skeletal Mesh Component Object Reference=GetPartSkeletalMeshComponent()) -> (Output_Get:Skeletal Mesh Component Object Reference)
   Event BlueprintUpdateAnimation()
   > n1 = Sequence!(execute) -> (then_0, then_1, then_2:Exec, then_3:Exec, then_4:Exec, then_5:Exec, then_6:Exec, then_7:Exec)
   > n2 = Branch(Condition=Get bIsTargeting())
   > "Set ADS Alpha "("ADS Alpha ":double=Get HeadRotBackAlpha()) -> (Output_Get:double)
   "Set ADS Alpha "(execute=@n2.else, "ADS Alpha "=0)
   n3 = Branch(execute=@n1.then_1, Condition=OR Boolean(A=Get bIsJumping(), B=Get bIsFalling()))
   > Set JumpingFalling(JumpingFalling:double=1) -> (Output_Get:double)
   Set JumpingFalling(execute=@n3.else, JumpingFalling=0)
   n4 = BreakTransform(InTransform=GetDeltaTransformFromRefPose(self=GetOwnerSkeletalMeshComponent(), BoneName=head, BaseName=spine_05))
   Set Head Down(execute=@n1.then_2, Head Down:double=MapRangeClamped(Value=BreakRotator(InRot=@n4.Rotation).Yaw, InRangeA=3, InRangeB=20)) -> (Output_Get:double)
   > Set IsEmoting(IsEmoting:double="To Float (Boolean)"(InBool=Get bIsPlayingEmote())) -> (Output_Get:double)
   Set Swing(execute=@n1.then_3, Swing:double=SelectFloat(A=Get HeadSwingAlpha(), B=0, bPickA=Get bIsPlayingMeleeAnim())) -> (Output_Get:double)
   n5 = Get bIsSkyDiving()
   Set SkyDive_Front(execute=@n1.then_4, SkyDive_Front:double=SelectFloat(A=Get SkydivingFrontAlpha(), B=0, bPickA=@n5)) -> (Output_Get:double)
   Set SkyDive_Back(execute=@n1.then_5, SkyDive_Back:double=SelectFloat(A=Get SkydivingBackAlpha(), B=0, bPickA=@n5)) -> (Output_Get:double)
   n6 = Branch(execute=@n1.then_7, Condition=AND Boolean!(A=Get bIsInFrontEnd(), B=Get bIsPlayingEmote(), C:bool=Get bRigidBodyEnabled()) -> (ReturnValue))
   > n7 = Sequence!(execute) -> (then_0, then_1, then_2)
   > Do Once(Reset=@n9.then_0)
   n8 = Get SkelMeshOwn() -> (SkelMeshOwn:Skeletal Mesh Component Object Reference)
   > SetSkinWeightProfile!(execute, self:Skinned Mesh Component Object Reference=@n8, InProfileName:Name=Alternative_01, InLayer:ESkinWeightProfileLayer Enum=Secondary) -> (then, ReturnValue:bool)
   Do Once(execute=@n9.then_1, Reset=@n7.then_1)
   > ClearSkinWeightProfile!(execute, self:Skinned Mesh Component Object Reference=@n8, InLayer:ESkinWeightProfileLayer Enum=Secondary) -> (then)
   n9 = Sequence(execute=@n6.else)
   n10 = Branch(Condition=Get bIsTargeting())
   n11 = BreakTransform(InTransform=GetDeltaTransformFromRefPose(self=GetOwnerSkeletalMeshComponent(), BoneName=head, BaseName=spine_05))
   > "Set ADS Alpha "("ADS Alpha "=MapRangeClamped(Value=BreakRotator(InRot=@n11.Rotation).Yaw, InRangeA=-1, InRangeB=-20))
   "Set ADS Alpha "(execute=@n10.else, "ADS Alpha "=0)
   SelectFloat(A=0, B=0, bPickA=false)
  ]}
]
variables [
 "ADS Alpha " {type=number}
 Head Down {type=number}
 IsEmoting {type=number}
 JumpingFalling {type=number}
 SkelMeshOwn {type=SkeletalMeshComponent}
 SkyDive_Back {type=number}
 SkyDive_Front {type=number}
 Swing {type=number}
]
